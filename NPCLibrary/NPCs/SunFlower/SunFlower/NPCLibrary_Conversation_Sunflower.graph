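# Conversation database for the Sunflower NPC, a farm-themed character who greets the player with speech and leads into shared conversation options.
# A single entry point feeds one speech node with a default line about rumbling ground.
# Contextual lines can replace it for a first-ever conversation or specific player outfits, ranked by curve-table priorities.
# The speech then branches through link nodes to the shared NPC library quests and basic services conversations.
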ConversationDatabase /NPCLibrary/NPCs/SunFlower/SunFlower/NPCLibrary_Conversation_Sunflower
entry_points [{tag=Conversation.Entry.Sunflower, destinations [#1]}]
nodes [
 ConversationEntryPointNode_7 {
  id=#1
  class=ConversationEntryPointNode
  outputs [#2]
  properties {entryTag {tagName=Conversation.Entry.Sunflower}}
 }
 Speech_C_4 {
  id=#2
  class=Speech_C
  outputs [#3, #4]
  properties {
   nodeName=""
   generalConfig {
    defaultMessage="Ground's been rumblin', spooking my animals. My poor chickens won't even cross the road."
    contextualMessages [
     {
      message="This old farm could use some upgrades. Like a new silo. Or an electric catapult!"
      contextRequirements [{requirement {class=FortControllerRequirement_IsFirstEverConversationWithNPC}}]
      priority {curve {rowName=Tandem.Speech.Priority.FirstEverInteraction}}
     }
     {
      message="I've got some big ideas for the farm, sir."
      contextRequirements [{requirement {class=FortControllerRequirement_HasCID, properties {allowedCIds [&Character_Citadel]}}}]
      priority {curve {rowName=Tandem.Speech.Priority.OutfitReaction}}
     }
     {
      message=You get outta here before I call the Champion on ya.
      contextRequirements [{requirement {
         class=FortControllerRequirement_HasCID
         properties {allowedCIds [/BRCosmetics/Athena/Items/Cosmetics/Characters/CID_714_Athena_Commando_M_AnarchyAcresFarmer]}
        }}]
      priority {curve {rowName=Tandem.Speech.Priority.OutfitReaction}}
     }
     {
      message="How do you feel about \"acquiring\" me a plasmacore tractor?"
      contextRequirements [{requirement {class=FortControllerRequirement_HasCID, properties {allowedCIds [&Character_VitalPsych]}}}]
      priority {curve {rowName=Tandem.Speech.Priority.OutfitReaction}}
     }
     {
      message=You and your Luminaries got any cool prototypes for farm equipment?
      contextRequirements [{requirement {class=FortControllerRequirement_HasCID, properties {allowedCIds [/BRCosmetics/Athena/Items/Cosmetics/Characters/Character_Nebula]}}}]
      priority {curve {rowName=Tandem.Speech.Priority.OutfitReaction}}
     }
     {
      message="Uh, Mr. Destroyer sir, this farm isn't available for conquering."
      contextRequirements [{requirement {class=FortControllerRequirement_HasCID, properties {allowedCIds [&Character_Inferno]}}}]
      priority {curve {rowName=Tandem.Speech.Priority.OutfitReaction}}
     }
    ]
   }
  }
 }
 ConversationLinkNode_0 {id=#4, class=ConversationLinkNode, properties {remoteEntryTag {tagName=Conversation.Entry.BasicServices}}}
 ConversationLinkNode_1 {id=#3, class=ConversationLinkNode, properties {remoteEntryTag {tagName=Conversation.Entry.NPCLibrary.Quests}}}
]
